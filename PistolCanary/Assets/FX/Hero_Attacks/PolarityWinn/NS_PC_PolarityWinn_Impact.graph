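# Niagara impact burst for PolarityWinn's hero attack, combining a flare with dark glow, expanding rings, sparks, mesh shockwave volumes and a light flash.
# User.ExplosionDelay times every burst except Light001, while System.ExplosionScale sizes the flare, rings and meshes.
# User.LeftRightHand blends those layers between purple-magenta and cyan-blue palettes, and System.EmpowerAlpha drives a flare material parameter.
# Light001 keeps fixed colors and scales its sprite by camera-distance LOD.
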NiagaraSystem /PistolCanary/Assets/FX/Hero_Attacks/PolarityWinn/NS_PC_PolarityWinn_Impact
dependencies {
 usedDataInterfaces [
  &NiagaraDataInterfaceCurve
  &NiagaraDataInterfaceColorCurve
  &NiagaraDataInterfaceVectorCurve
  &NiagaraDataInterfacePlatformSet
  &NiagaraDataInterfaceArrayFloat
 ]
 usedDynamicInputs [
  &SelectFloatFromArray
  &Add_Float
  &SetBoolByFloatComparison
 ]
 usedModules [
  &self:SystemSpawnScript.SystemScriptSource.SystemScriptGraph.NiagaraNodeAssignment_3.SetVariables_#1
  &self:SystemSpawnScript.SystemScriptSource.SystemScriptGraph.NiagaraNodeAssignment_0.SetVariables_#2
  &self:SystemSpawnScript.SystemScriptSource.SystemScriptGraph.NiagaraNodeAssignment_6.SetVariables_#3
  &SystemState
  &self:SystemSpawnScript.SystemScriptSource.SystemScriptGraph.NiagaraNodeAssignment_5.SetVariables_#4
 ]
}
emitters {
 Light001 {
  emitterUpdateScript [
   &EmitterState()
   &SpawnBurst_Instantaneous(Spawn Count=1, Use Spawn Probability=false)
  ]
  particleSpawnScript [
   &InitializeParticle(
    Color=<rgba 4.0101 1.0075 5 1>
    Lifetime=0.5
    Sprite Size Mode=::NewEnumerator1
    Uniform Sprite Size Max=5000
    Uniform Sprite Size Min=5000
    UsePositionOffset=false
    Write Material Random=true
   )
   /SetVariables(Particles.SpriteSize=&Multiply_Vector2D_ByFloat(
     Float=&RemapRange(
      Input Value To Remap=&FN_LOD_Distance_Remap(Start Distance=5000, End Distance=15000, Minimum Output=1, Maximum Output=3)
      Input Min=0
      Input Max=1
      Output Min=0
      Output Max=2
     )
     Vector2D=<5000 5000>
    )) {isSetParameters=true}
   &SystemLocation(MaskSpawnGroup=false)
   /SetVariables(Particles.LightPos:Vector3f=<0 0 500>) {isSetParameters=true}
  ]
  particleUpdateScript [
   &ParticleState()
   &CameraOffset(Camera Offset Amount=100)
   &ScaleColor(
    Scale Mode=::NewEnumerator1
    Scale RGBA=&MakeVector4_FromLinearColor(LinearColor=&LerpLinearColors(StartColor=<rgba 0.0957 0.5612 1 1>, EndColor=<rgba 1.2268 0.3389 2 1>, LerpFactor=1))
    ScaleRGBA=true
   )
   &ScaleSpriteSize(Uniform Curve Sprite Scale=DataInterface() {propertyValues {Curve {keys [
        {arriveTangent=1.1934, interpMode=RCIM_Cubic, leaveTangent=1.1934, tangentMode=RCTM_User, time=0, value=0}
        {arriveTangent=0.4403, interpMode=RCIM_Cubic, leaveTangent=0.4403, tangentMode=RCTM_User, time=0.8, value=0.85}
        {arriveTangent=-0.0405, interpMode=RCIM_Cubic, leaveTangent=-0.0405, tangentMode=RCTM_User, time=1.3, value=1}
       ]}}})
   &DynamicMaterialParameters(
    Index 0 Param 1=0.5
    Index 0 Param 2=1
    Index 1 Param 1=500
    Module.Param0WriteEnabled=true
    Module.Param0WriteEnabled1=true
    Module.Param1WriteEnabled=true
    Module.Param1WriteEnabled1=true
    Module.Param2WriteEnabled=true
    Module.Param2WriteEnabled1=true
    Module.Param3WriteEnabled=true
    Module.Param3WriteEnabled1=true
    Write Parameter Index 1=true
   )
   /SetVariables(
    Particles.LightColor=&ColorFromCurve(ColorCurve=DataInterface() {propertyValues {
       AlphaCurve {keys [<at 0 1 RCIM_Cubic>,<at 1 0 RCIM_Cubic>]}
       BlueCurve {keys [{arriveTangent=-499, interpMode=RCIM_Cubic, time=0, value=100}, {interpMode=RCIM_Cubic, leaveTangent=-499, time=1, value=1}]}
       GreenCurve {keys [{arriveTangent=-6.2674, interpMode=RCIM_Cubic, time=0, value=6.459}, {interpMode=RCIM_Cubic, leaveTangent=-6.2674, time=1, value=0.2897}]}
       RedCurve {keys [{arriveTangent=1, interpMode=RCIM_Cubic, time=0, value=3.434}, {interpMode=RCIM_Cubic, leaveTangent=1, time=1, value=0.6258}]}
      }})
    Particles.LightIntensity_Component:NiagaraFloat=&FloatFromCurve(FloatCurve=DataInterface() {propertyValues {Curve {keys [
         {interpMode=RCIM_Cubic, tangentMode=RCTM_User, time=0, value=0}
         {interpMode=RCIM_Cubic, tangentMode=RCTM_User, time=0.1, value=0.7}
         <at 1 0>
        ]}}})
    Particles.LightRadius=&FloatFromCurve(
     FloatCurve=DataInterface() {propertyValues {Curve {keys [{interpMode=RCIM_Cubic, tangentMode=RCTM_User, time=0, value=0}, {interpMode=RCIM_Cubic, tangentMode=RCTM_User, time=0.1, value=1}]}}}
     Scale Curve=100
    )
   ) {isSetParameters=true}
  ]
  rendererClasses [
   &NiagaraSpriteRendererProperties
   &NiagaraLightRendererProperties
   &NiagaraComponentRendererProperties
  ]
  renderers [
   {rendererClass=&NiagaraSpriteRendererProperties}
   {rendererClass=&NiagaraLightRendererProperties, rendererIndex=1}
   {rendererClass=&NiagaraComponentRendererProperties, rendererIndex=2}
  ]
 }
 NE_DarkFlare {
  emitterUpdateScript [
   &EmitterState()
   &SpawnBurst_Instantaneous(Spawn Time=$User.ExplosionDelay, Use Spawn Probability=false, UseLoopCountLimit=false)
  ]
  particleSpawnScript [
   &InitializeParticle(
    Color=<rgba 0.6771 0.6458 1 1>
    Lifetime=0.5
    Mass Mode=::NewEnumerator1
    Mesh Scale Mode=::NewEnumerator3
    Mesh Uniform Scale=0.2
    Position=&SimulationPosition()
    Position Mode=::NewEnumerator1
    Position Offset=$System.SystemOffset
    Sprite Rotation Mode=::NewEnumerator2
    Sprite Size=&Vector2DFromFloat(Value=&Multiply_Float(A=450, B=$System.ExplosionScale))
    Sprite Size Mode=::NewEnumerator4
    UsePositionOffset=true
    Write Material Random=true
   ) {module=InitializeParticle001}
   &CameraOffset(Camera Offset Amount=75)
  ]
  particleUpdateScript [
   &ParticleState()
   &SolveForcesAndVelocity(Clamp Velocity=false, Limit Acceleration=false)
   &ScaleSpriteSize(Scale Factor=&Multiply_Vector2D_ByFloat(
     Float=&FloatFromCurve(
      FloatCurve=DataInterface() {propertyValues {Curve {keys [
          {arriveTangent=0.7585, interpMode=RCIM_Cubic, leaveTangent=0.7585, tangentMode=RCTM_User, time=0, value=0.4}
          {interpMode=RCIM_Cubic, tangentMode=RCTM_User, time=1, value=1}
         ]}}}
      Scale Curve=0.1
     )
     Vector2D=<1 1>
    ))
   &SpriteRotationRate(Rotation Rate=&UniformRangedFloat(Minimum=-50, Maximum=50))
   /SetVariables(Particles.DarkGlowSize:Vector2f=&Multiply_Vector2D_ByFloat(Float=3, Vector2D=$Particles.SpriteSize)) {isSetParameters=true}
   /SetVariables(Particles.DargGlowColor:LinearColor=&MakeLinearColorFromVectorAndFloat(
     "Vector (RGB)"=&MakeVectorFromLinearColorRGB(Color=<rgba 0.0452 0.0406 0.05 1>)
     "Float (Alpha)"=&FloatFromCurve(FloatCurve=DataInterface() {propertyValues {Curve {keys [
          {arriveTangent=-0.0975, interpMode=RCIM_Cubic, leaveTangent=-0.0975, tangentMode=RCTM_User, time=0, value=1}
          {interpMode=RCIM_Cubic, leaveTangent=0.002, time=1, value=0}
         ]}}})
    )) {isSetParameters=true}
   /SetVariables(Particles.CentralFlareSize:Vector2f=&Multiply_Vector2D_ByFloat(Float=3, Vector2D=$Particles.SpriteSize)) {isSetParameters=true}
   /SetVariables(Particles.CentralFlareColor:LinearColor=&MakeLinearColorFromVectorAndFloat(
     "Vector (RGB)"=&MakeVectorFromLinearColorRGB(Color=&LerpLinearColors(
       StartColor=&ColorFromCurve(ColorCurve=DataInterface() {propertyValues {
          AlphaCurve {keys [<at 0 1 RCIM_Cubic>,<at 1 0 RCIM_Cubic>]}
          BlueCurve {keys [<at 0 3>, {interpMode=RCIM_Cubic, leaveTangent=1.3364, time=1, value=1}]}
          GreenCurve {keys [<at 0 0>, {interpMode=RCIM_Cubic, leaveTangent=-5.7627, time=1, value=0.3351}]}
          RedCurve {keys [<at 0 1.7195>, {interpMode=RCIM_Cubic, leaveTangent=-13.364, time=1, value=0.8022}]}
         }})
       EndColor=&ColorFromCurve(ColorCurve=DataInterface() {propertyValues {
          AlphaCurve {keys [<at 0 1 RCIM_Cubic>,<at 1 0 RCIM_Cubic>]}
          BlueCurve {keys [<at 0 3>, {interpMode=RCIM_Cubic, leaveTangent=1.3364, time=1, value=1}]}
          GreenCurve {keys [<at 0 1.7537>, {interpMode=RCIM_Cubic, leaveTangent=-5.7627, time=1, value=0.2803}]}
          RedCurve {keys [<at 0 0.3>, {interpMode=RCIM_Cubic, leaveTangent=-13.364, time=1, value=0.2191}]}
         }})
       LerpFactor=$User.LeftRightHand
      ))
     "Float (Alpha)"=&FloatFromCurve(
      FloatCurve=DataInterface() {propertyValues {Curve {keys [
          <at 0 0 RCIM_Cubic>
          {arriveTangent=0.3854, interpMode=RCIM_Cubic, leaveTangent=0.3854, tangentMode=RCTM_User, time=0.1, value=1}
          <at 1 0 RCIM_Cubic>
         ]}}}
      Scale Curve=0.5
     )
    )) {isSetParameters=true}
   &ScaleColor(
    Scale Alpha=&FloatFromCurve(
     FloatCurve=DataInterface() {propertyValues {Curve {keys [
         {arriveTangent=6.4019, interpMode=RCIM_Cubic, leaveTangent=6.4019, tangentMode=RCTM_User, time=0, value=1}
         {interpMode=RCIM_Cubic, leaveTangent=0.002, time=1, value=0}
        ]}}}
     Scale Curve=0.02
    )
    Scale RGB=&MakeVectorFromLinearColorRGB(Color=$Particles.CentralFlareColor)
    ScaleA=true
    ScaleRGB=true
   )
   &DynamicMaterialParameters(
    Index 1 Param 1=0.3
    Index 1 Param 4=&Lerp_Float(A=0, B=1, Alpha=$System.EmpowerAlpha)
    Module.Index1Float0Write=true
    Module.Index1Float1Write=true
    Module.Index1Float2Write=true
    Module.Index1Float3Write=true
    Module.Index3Float0Write=true
    Module.Index3Float1Write=true
    Module.Index3Float2Write=true
    Module.Index3Float3Write=true
    Write Parameter Index 0=::NewEnumerator0
    Write Parameter Index 1=::NewEnumerator1
    Write Parameter Index 3=::NewEnumerator1
   )
   &FN_Sprite_Attributes(
    Create Sprite Rotation Attribute=true
    Create SpriteSize Attribute=true
    SpriteSize=&Multiply_Vector2D(A=$Particles.CentralFlareSize, B=<2 0.15>)
    Write to Module Namespace=true
   )
  ]
  rendererClasses [&NiagaraSpriteRendererProperties]
  renderers [
   {rendererClass=&NiagaraSpriteRendererProperties}
   {rendererClass=&NiagaraSpriteRendererProperties, rendererIndex=1}
   {rendererClass=&NiagaraSpriteRendererProperties, rendererIndex=2}
  ]
 }
 NE_SoundRings {
  emitterUpdateScript [
   &EmitterState()
   &SpawnBurst_Instantaneous(Spawn Count=4, Spawn Time=$User.ExplosionDelay, Use Spawn Probability=false, UseLoopCountLimit=false)
  ]
  particleSpawnScript [
   &InitializeParticle(
    Lifetime=0.5
    Position Offset=$System.SystemOffset
    Sprite Size=&Multiply_Vector2D_ByFloat(
     Float=&Multiply_Float(
      A=&Lerp_Float(A=25, B=150, Alpha=&ReturnNormalizedExecIndex())
      B=$System.ExplosionScale
     )
     Vector2D=<1 1>
    )
    Sprite Size Mode=::NewEnumerator4
    UsePositionOffset=true
    Write Material Random=true
   )
   /SetVariables(Particles.BaseNormalisedIndex:NiagaraFloat=&ReturnNormalizedExecIndex()) {isSetParameters=true}
   /SetVariables(Particles.BurstDirection:Vector3f=<1 0 0>) {isSetParameters=true}
   /SetVariables(Particles.Age:NiagaraFloat=&Multiply_Float(A=&ReturnNormalizedExecIndex(), B=-0.1)) {isSetParameters=true}
   &CameraOffset(Camera Offset Amount=-50)
  ]
  particleUpdateScript [
   &ParticleState()
   /SetVariables(Particles.Position=&AddVectorToPosition(
     Position=$Particles.Initial.Position
     Vector=&Multiply_VectorByFloat(
      Vector=&Multiply_VectorByFloat(Vector=$Particles.BurstDirection, Float=&Multiply_Float(A=-10, B=$Particles.BaseNormalisedIndex))
      Float=&FloatFromCurve(FloatCurve=DataInterface() {propertyValues {Curve {keys [
           {arriveTangent=1.8928, interpMode=RCIM_Cubic, leaveTangent=1.8928, tangentMode=RCTM_User, time=0, value=0}
           {arriveTangent=-0.5182, interpMode=RCIM_Cubic, leaveTangent=-0.5182, tangentMode=RCTM_User, time=1, value=1}
          ]}}})
     )
    )) {isSetParameters=true}
   &ScaleSpriteSize(
    Scale Sprite Size Mode=::NewEnumerator0
    Uniform Scale Factor=&Multiply_Float(
     A=&FloatFromCurve(FloatCurve=DataInterface() {propertyValues {Curve {keys [
          {arriveTangent=1.8909, interpMode=RCIM_Cubic, leaveTangent=1.8909, tangentMode=RCTM_User, time=0, value=0}
          {interpMode=RCIM_Cubic, tangentMode=RCTM_User, time=1, value=1}
         ]}}})
     B=1
    )
   )
   &DynamicMaterialParameters(
    Index 0 Param 1=0.2
    Index 0 Param 2=1
    Module.Index0Float0Write=true
    Module.Index0Float1Write=true
    Module.Index0Float2Write=true
    Module.Index0Float3Write=true
   )
   &ScaleColor(
    Scale Alpha=&FloatFromCurve(
     FloatCurve=DataInterface() {propertyValues {Curve {keys [
         {arriveTangent=4.2862, interpMode=RCIM_Cubic, leaveTangent=4.2862, tangentMode=RCTM_User, time=0, value=0}
         {interpMode=RCIM_Cubic, tangentMode=RCTM_User, time=0.1, value=1}
         {arriveTangent=-0.5134, interpMode=RCIM_Cubic, leaveTangent=-0.5134, tangentMode=RCTM_User, time=1, value=0}
        ]}}}
     Scale Curve=0.2
    )
    Scale RGB=&MakeVectorFromLinearColorRGB(Color=&Multiply_LinearColorByFloat(
      Linear Color=&LerpLinearColors(
       StartColor=&ColorFromCurve(
        ColorCurve=DataInterface() {propertyValues {
          AlphaCurve {keys [<at 0 1 RCIM_Cubic>]}
          BlueCurve {keys [<at 0 10 RCIM_Cubic>,<at 1 2 RCIM_Cubic>]}
          GreenCurve {keys [<at 0 0.024 RCIM_Cubic>,<at 1 0.8077 RCIM_Cubic>]}
          RedCurve {keys [<at 0 1.7023 RCIM_Cubic>,<at 1 1.8548 RCIM_Cubic>]}
         }}
        CurveIndex=$Particles.BaseNormalisedIndex
       )
       EndColor=&ColorFromCurve(
        ColorCurve=DataInterface() {propertyValues {
          AlphaCurve {keys [<at 0 1 RCIM_Cubic>]}
          BlueCurve {keys [<at 0 10 RCIM_Cubic>,<at 1 2 RCIM_Cubic>]}
          GreenCurve {keys [<at 0 7.0937 RCIM_Cubic>,<at 1 0.2845 RCIM_Cubic>]}
          RedCurve {keys [<at 0 0 RCIM_Cubic>,<at 1 0.3733 RCIM_Cubic>]}
         }}
        CurveIndex=$Particles.BaseNormalisedIndex
       )
       LerpFactor=$User.LeftRightHand
      )
      Float=&Lerp_Float(
       A=0
       B=5
       Alpha=&FloatFromCurve(FloatCurve=DataInterface() {propertyValues {Curve {keys [
            {arriveTangent=-1.6896, interpMode=RCIM_Cubic, leaveTangent=-1.6896, tangentMode=RCTM_User, time=0, value=1}
            {arriveTangent=0.0045, interpMode=RCIM_Cubic, leaveTangent=0.0045, tangentMode=RCTM_User, time=1, value=0}
           ]}}})
      )
     ))
    ScaleA=true
    ScaleRGB=true
   )
   &SpriteFacingAndAlignment(Sprite Facing=$Particles.BurstDirection, WriteAlignment=false, WriteFacing=true)
   &FN_Sprite_Attributes(
    Color=&Multiply_LinearColorByFloat(
     Linear Color=&LerpLinearColors(
      StartColor=&ColorFromCurve(
       ColorCurve=DataInterface() {propertyValues {
         AlphaCurve {keys [<at 0 1 RCIM_Cubic>]}
         BlueCurve {keys [<at 0 10 RCIM_Cubic>,<at 1 2 RCIM_Cubic>]}
         GreenCurve {keys [<at 0 0.024 RCIM_Cubic>,<at 1 0.8077 RCIM_Cubic>]}
         RedCurve {keys [<at 0 1.7023 RCIM_Cubic>,<at 1 1.8548 RCIM_Cubic>]}
        }}
       CurveIndex=$Particles.BaseNormalisedIndex
      )
      EndColor=&ColorFromCurve(
       ColorCurve=DataInterface() {propertyValues {
         AlphaCurve {keys [<at 0 1 RCIM_Cubic>]}
         BlueCurve {keys [<at 0 10 RCIM_Cubic>,<at 1 2 RCIM_Cubic>]}
         GreenCurve {keys [<at 0 7.0937 RCIM_Cubic>,<at 1 0.2845 RCIM_Cubic>]}
         RedCurve {keys [<at 0 0 RCIM_Cubic>,<at 1 0.3733 RCIM_Cubic>]}
        }}
       CurveIndex=$Particles.BaseNormalisedIndex
      )
      LerpFactor=$User.LeftRightHand
     )
     Float=&Lerp_Float(
      A=0
      B=5
      Alpha=&FloatFromCurve(FloatCurve=DataInterface() {propertyValues {Curve {keys [
           {arriveTangent=-1.6896, interpMode=RCIM_Cubic, leaveTangent=-1.6896, tangentMode=RCTM_User, time=0, value=1}
           {arriveTangent=0.0045, interpMode=RCIM_Cubic, leaveTangent=0.0045, tangentMode=RCTM_User, time=1, value=0}
          ]}}})
     )
    )
    Create Color Attribute=true
    Create Position Attribute=true
    Create SpriteSize Attribute=true
    Position:NiagaraPosition=&AddVectorToPosition(
     Position=$Particles.Initial.Position
     Vector=&Multiply_VectorByFloat(
      Vector=$Particles.BurstDirection
      Float=&FloatFromCurve(
       FloatCurve=DataInterface() {propertyValues {Curve {keys [
           {arriveTangent=1.8928, interpMode=RCIM_Cubic, leaveTangent=1.8928, tangentMode=RCTM_User, time=0, value=0}
           {arriveTangent=0.3918, interpMode=RCIM_Cubic, leaveTangent=0.3918, tangentMode=RCTM_User, time=1, value=1}
          ]}}}
       Scale Curve=&Lerp_Float(A=0, B=10, Alpha=&ReturnNormalizedExecIndex())
      )
     )
    )
    SpriteSize=&Vector2DFromFloat(Value=&Multiply_Float(
      A=&Multiply_Float(A=&Lerp_Float(A=50, B=5, Alpha=$Particles.BaseNormalisedIndex), B=$System.ExplosionScale)
      B=&FloatFromCurve(FloatCurve=DataInterface() {propertyValues {Curve {keys [
           {arriveTangent=2.3035, interpMode=RCIM_Cubic, leaveTangent=2.3035, tangentMode=RCTM_User, time=0, value=0}
           {interpMode=RCIM_Cubic, tangentMode=RCTM_User, time=1, value=1}
          ]}}})
     ))
    Write to Module Namespace=true
   )
   &InheritSourceMovement(
    Applied Position Scale=&VectorFromFloat(Value=&Multiply_Float(A=1, B=-1))
    ApplyForce=false
    ApplyPosition=true
    ApplyPositionFalloff=false
   )
   &SolveForcesAndVelocity(Clamp Velocity=false, Limit Acceleration=false)
  ]
  rendererClasses [&NiagaraSpriteRendererProperties, &NiagaraLightRendererProperties]
  renderers [
   {rendererClass=&NiagaraSpriteRendererProperties}
   {rendererClass=&NiagaraSpriteRendererProperties, rendererIndex=1}
   {rendererClass=&NiagaraLightRendererProperties, rendererIndex=2}
  ]
 }
 NE_Sprites {
  emitterUpdateScript [
   &EmitterState()
   &SpawnBurst_Instantaneous(Spawn Count=20, Spawn Time=$User.ExplosionDelay, Use Spawn Probability=false, UseLoopCountLimit=false)
   &SpawnBurst_Instantaneous(Spawn Count=5, Spawn Group=1, Spawn Time=$User.ExplosionDelay, Use Spawn Probability=false, UseLoopCountLimit=false) {module=SpawnBurst_Instantaneous001}
  ]
  particleSpawnScript [
   &InitializeParticle(
    Lifetime=&RandomRangeFloat(Minimum=0.5, Maximum=0.8)
    Sprite Size=&Multiply_Vector2D_ByFloat(Float=&RandomRangeFloat(Minimum=100, Maximum=50), Vector2D=<1 3>)
    Sprite Size Mode=::NewEnumerator4
    UsePositionOffset=false
    Write Material Random=true
   )
   /SetVariables(Particles.VisibilityTag=&SpawnGroup()) {isSetParameters=true}
   &ShapeLocation(
    MaskSpawnGroup=false
    Override Seed=true
    Random Seed:NiagaraInt32=&SpawnGroup()
    Sphere Radius=2
    UseOwnerScale=true
   )
   &AddVelocity(
    Cone Angle=70
    Module.UseConeFalloff=false
    Module.UseVelDistribution=false
    Override Seed=false
    Velocity Mode=::NewEnumerator2
    Velocity Speed=&RandomRangeFloat(Minimum=5000, Maximum=10000)
   )
  ]
  particleUpdateScript [
   &ParticleState()
   &InheritSourceMovement(
    Applied Position Scale=&VectorFromFloat(Value=&Multiply_Float(
      A=1
      B=&FloatFromCurve(FloatCurve=DataInterface() {propertyValues {Curve {keys [
           {arriveTangent=-1.9407, interpMode=RCIM_Cubic, leaveTangent=-1.9407, tangentMode=RCTM_User, time=0, value=1}
           {interpMode=RCIM_Cubic, tangentMode=RCTM_User, time=1, value=0.5}
          ]}}})
     ))
    ApplyForce=false
    ApplyPosition=true
    ApplyPositionFalloff=false
   )
   &Drag(Drag=&RandomRangeFloat(Minimum=3, Maximum=5), Use Linear Drag=true, Use Rotational Drag=true)
   &ScaleColor(
    Scale Alpha=&FloatFromCurve(FloatCurve=DataInterface() {propertyValues {Curve {keys [
         {arriveTangent=6.4019, interpMode=RCIM_Cubic, leaveTangent=6.4019, tangentMode=RCTM_User, time=0, value=0}
         <at 0.1 0.7 RCIM_Cubic>
         <at 1 0 RCIM_Cubic>
        ]}}})
    Scale RGB=&MakeVectorFromLinearColorRGB(Color=&MultiplyLinearColor(
      Linear Color=&ColorFromCurve(ColorCurve=DataInterface() {propertyValues {
         AlphaCurve {keys [<at 0 1 RCIM_Cubic>]}
         BlueCurve {keys [<at 0 10 RCIM_Cubic>,<at 0.3726 1 RCIM_Cubic>]}
         GreenCurve {keys [<at 0 10 RCIM_Cubic>,<at 0.3726 1 RCIM_Cubic>]}
         RedCurve {keys [<at 0 10 RCIM_Cubic>,<at 0.3726 1 RCIM_Cubic>]}
        }})
      Scale Factor=&MakeVector4_FromLinearColor(LinearColor=&LerpLinearColors(
        StartColor=&ColorFromCurve(
         ColorCurve=DataInterface() {propertyValues {
           AlphaCurve {keys [<at 0 1 RCIM_Cubic>]}
           BlueCurve {keys [<at 0.3173 1 RCIM_Cubic>,<at 0.3548 1>,<at 0.6928 1>,<at 0.7245 0.6437>]}
           GreenCurve {keys [<at 0.3173 0.0008 RCIM_Cubic>,<at 0.3548 0>,<at 0.6928 0>,<at 0.7245 0>]}
           RedCurve {keys [<at 0.3173 0.2843 RCIM_Cubic>,<at 0.3548 0.508>,<at 0.6928 0.8721>,<at 0.7245 1>]}
          }}
         CurveIndex=&RandomRangeFloat(Minimum=0, Maximum=1)
        )
        EndColor=&ColorFromCurve(
         ColorCurve=DataInterface() {propertyValues {
           AlphaCurve {keys [<at 0 1 RCIM_Cubic>]}
           BlueCurve {keys [<at 0.3173 1 RCIM_Cubic>,<at 0.3548 1>,<at 0.6928 1>,<at 0.7245 1>]}
           GreenCurve {keys [<at 0.3173 0.2286 RCIM_Cubic>,<at 0.3548 0.4077>,<at 0.6928 0.4077>,<at 0.7245 0.9502>]}
           RedCurve {keys [<at 0.3173 0.1159 RCIM_Cubic>,<at 0.3548 0.1>,<at 0.6928 0.1>,<at 0.7245 0.1>]}
          }}
         CurveIndex=&RandomRangeFloat(Minimum=0, Maximum=1)
        )
        LerpFactor=$User.LeftRightHand
       ))
     ))
    ScaleA=true
    ScaleRGB=true
   )
   &SolveForcesAndVelocity(Clamp Velocity=false, Limit Acceleration=false)
   &DynamicMaterialParameters(
    Index 0 Param 1=2
    Index 0 Param 3=1
    Index 3 Param 1=&Lerp_Float(
     A=0.2
     B=1
     Alpha=&FloatFromCurve(FloatCurve=DataInterface() {propertyValues {Curve {keys [<at 0 0>,<at 1 1>]}}})
    )
    Module.Index0Float0Write=true
    Module.Index0Float1Write=true
    Module.Index0Float2Write=true
    Module.Index0Float3Write=true
    Module.Index3Float0Write=true
    Module.Index3Float1Write=true
    Module.Index3Float2Write=true
    Module.Index3Float3Write=true
    Write Parameter Index 3=::NewEnumerator1
   )
   /SetVariables(Particles.SparkGlowColor:LinearColor=&MultiplyLinearColor(
     Linear Color=<rgba 0.1 0.1 0.1 0.05>
     Scale Factor=&MakeVector4_FromLinearColor(LinearColor=$Particles.Color)
    )) {isSetParameters=true}
   /SetVariables(Particles.SparkGlowSize:Vector2f=&Multiply_Vector2D(A=$Particles.SpriteSize, B=<5 2>)) {isSetParameters=true}
   /SetVariables(Particles.SparkGlowDynMat:Vector4f=<2 20 90 1>) {isSetParameters=true}
   &FN_Sprite_Attributes(
    Color=&MakeLinearColorFromVectorAndFloat(
     "Vector (RGB)"=&MakeVectorFromLinearColorRGB(Color=&Multiply_LinearColorByFloat(
       Linear Color=&LerpLinearColors(
        StartColor=&ColorFromCurve(
         ColorCurve=DataInterface() {propertyValues {
           AlphaCurve {keys [<at 0 1 RCIM_Cubic>]}
           BlueCurve {keys [<at 0.3173 1 RCIM_Cubic>,<at 0.3548 1>,<at 0.6928 1>,<at 0.7245 0.6437>]}
           GreenCurve {keys [<at 0.3173 0.0008 RCIM_Cubic>,<at 0.3548 0>,<at 0.6928 0>,<at 0.7245 0>]}
           RedCurve {keys [<at 0.3173 0.2843 RCIM_Cubic>,<at 0.3548 0.508>,<at 0.6928 0.8721>,<at 0.7245 1>]}
          }}
         CurveIndex=&RandomRangeFloat(Minimum=0, Maximum=1)
        )
        EndColor=&ColorFromCurve(
         ColorCurve=DataInterface() {propertyValues {
           AlphaCurve {keys [<at 0 1 RCIM_Cubic>]}
           BlueCurve {keys [<at 0.3173 1 RCIM_Cubic>,<at 0.3548 1>,<at 0.6928 1>,<at 0.7245 1>]}
           GreenCurve {keys [<at 0.3173 0 RCIM_Cubic>,<at 0.3548 0.3419>,<at 0.6928 0.3419>,<at 0.7245 0.9447>]}
           RedCurve {keys [<at 0.3173 0.0183 RCIM_Cubic>,<at 0.3548 0>,<at 0.6928 0>,<at 0.7245 0.0002>]}
          }}
         CurveIndex=&RandomRangeFloat(Minimum=0, Maximum=1)
        )
        LerpFactor=$User.LeftRightHand
       )
       Float=&Lerp_Float(
        A=1
        B=20
        Alpha=&FloatFromCurve(FloatCurve=DataInterface() {propertyValues {Curve {keys [
             {arriveTangent=-0.4493, interpMode=RCIM_Cubic, leaveTangent=-0.4493, tangentMode=RCTM_User, time=0, value=1}
             {interpMode=RCIM_Cubic, tangentMode=RCTM_User, time=0.7, value=0}
            ]}}})
       )
      ))
     "Float (Alpha)"=&FloatFromCurve(FloatCurve=DataInterface() {propertyValues {Curve {keys [
          {interpMode=RCIM_Cubic, tangentMode=RCTM_User, time=0, value=1}
          {arriveTangent=-2.0246, interpMode=RCIM_Cubic, leaveTangent=-2.0246, tangentMode=RCTM_User, time=1, value=0}
         ]}}})
    )
    Create Color Attribute=true
    Create SpriteSize Attribute=true
    SpriteSize=&Multiply_Vector2D_ByFloat(
     Float=&Multiply_Float(
      A=&RandomRangeFloat(Minimum=0.5, Maximum=1.5)
      B=&FloatFromCurve(FloatCurve=DataInterface() {propertyValues {Curve {keys [
           {interpMode=RCIM_Cubic, tangentMode=RCTM_User, time=0.2, value=1}
           {arriveTangent=-2.4619, interpMode=RCIM_Cubic, leaveTangent=-2.4619, tangentMode=RCTM_User, time=1, value=0}
          ]}}})
     )
     Vector2D=<5 5>
    )
    Write to Module Namespace=true
   )
  ]
  rendererClasses [&NiagaraSpriteRendererProperties]
  renderers [
   {rendererClass=&NiagaraSpriteRendererProperties}
   {rendererClass=&NiagaraSpriteRendererProperties, rendererIndex=1}
   {rendererClass=&NiagaraSpriteRendererProperties, rendererIndex=2}
  ]
 }
 NE_VolumeSquare {
  emitterUpdateScript [
   &EmitterState()
   &SpawnBurst_Instantaneous(Spawn Time=$User.ExplosionDelay, Use Spawn Probability=false, UseLoopCountLimit=false)
  ]
  particleSpawnScript [&InitializeParticle(
    Color Mode=::NewEnumerator0
    Lifetime=0.3
    Mesh Scale Mode=::NewEnumerator3
    Position Offset=$System.SystemOffset
    UsePositionOffset=true
    Write Material Random=true
   )]
  particleUpdateScript [
   &ParticleState()
   &DynamicMaterialParameters(
    Index 0 Param 1=&Lerp_Float(
     A=1
     B=0.9
     Alpha=&FloatFromCurve(FloatCurve=DataInterface() {propertyValues {Curve {keys [
          {interpMode=RCIM_Cubic, tangentMode=RCTM_User, time=0, value=1}
          {arriveTangent=-1.8092, interpMode=RCIM_Cubic, leaveTangent=-1.8092, tangentMode=RCTM_User, time=1, value=0}
         ]}}})
    )
    Index 0 Param 3=1
    Index 0 Param 4=1
    Index 1 Param 2=1
    Module.Index0Float0Write=true
    Module.Index0Float1Write=true
    Module.Index0Float2Write=true
    Module.Index0Float3Write=true
    Module.Index1Float0Write=true
    Module.Index1Float1Write=true
    Module.Index1Float2Write=true
    Module.Index1Float3Write=true
    Write Parameter Index 1=::NewEnumerator1
   )
   &ScaleMeshSize(Scale Factor=&Multiply_VectorByFloat(
     Vector=&VectorFromCurve(
      VectorCurve=DataInterface() {propertyValues {
        XCurve {keys [
          {arriveTangent=1.3304, interpMode=RCIM_Cubic, leaveTangent=1.3304, tangentMode=RCTM_User, time=0, value=0}
          {interpMode=RCIM_Cubic, tangentMode=RCTM_User, time=1, value=1}
         ]}
        YCurve {keys [
          {arriveTangent=0.5179, interpMode=RCIM_Cubic, leaveTangent=0.5179, tangentMode=RCTM_User, time=0, value=0.8}
          {interpMode=RCIM_Cubic, tangentMode=RCTM_User, time=1, value=1}
         ]}
        ZCurve {keys [
          {arriveTangent=0.5179, interpMode=RCIM_Cubic, leaveTangent=0.5179, tangentMode=RCTM_User, time=0, value=0.8}
          {interpMode=RCIM_Cubic, tangentMode=RCTM_User, time=1, value=1}
         ]}
       }}
      Scale Curve=<0.75 0.5 0.5>
     )
     Float=&Multiply_Float(A=$System.ExplosionScale, B=0.5)
    ))
   &Color(Color=&Multiply_LinearColorByFloat(
     Linear Color=&MultiplyLinearColor(
      Linear Color=&ColorFromCurve(ColorCurve=DataInterface() {propertyValues {
         AlphaCurve {keys [<at 0 1 RCIM_Cubic>]}
         BlueCurve {keys [<at 0 10 RCIM_Cubic>,<at 0.5555 1 RCIM_Cubic>]}
         GreenCurve {keys [<at 0 6.0607 RCIM_Cubic>,<at 0.5555 1 RCIM_Cubic>]}
         RedCurve {keys [<at 0 4.8376 RCIM_Cubic>,<at 0.5555 1 RCIM_Cubic>]}
        }})
      Scale Factor=<1 1 1 1>
     )
     Float=&FloatFromCurve(
      FloatCurve=DataInterface() {propertyValues {Curve {keys [
          {arriveTangent=4.1951, interpMode=RCIM_Cubic, leaveTangent=4.1951, tangentMode=RCTM_User, time=0, value=0}
          {interpMode=RCIM_Cubic, tangentMode=RCTM_User, time=0.5, value=1}
          {arriveTangent=-3.2999, interpMode=RCIM_Cubic, leaveTangent=-3.2999, tangentMode=RCTM_User, time=1, value=0}
         ]}}}
      Scale Curve=0.3
     )
    ))
   &FN_Mesh_Attributes(
    Color=&Multiply_LinearColorByFloat(
     Linear Color=&MultiplyLinearColor(
      Linear Color=&LerpLinearColors(
       StartColor=&ColorFromCurve(ColorCurve=DataInterface() {propertyValues {
          AlphaCurve {keys [<at 0 1 RCIM_Cubic>]}
          BlueCurve {keys [<at 0 1 RCIM_Cubic>,<at 0.6468 1>,<at 1 1 RCIM_Cubic>]}
          GreenCurve {keys [<at 0 0.0009 RCIM_Cubic>,<at 0.6468 0.3222>,<at 1 0.3958 RCIM_Cubic>]}
          RedCurve {keys [<at 0 0.2466 RCIM_Cubic>,<at 0.6468 0.4166>,<at 1 0.9441 RCIM_Cubic>]}
         }})
       EndColor=&ColorFromCurve(ColorCurve=DataInterface() {propertyValues {
          AlphaCurve {keys [<at 0 1 RCIM_Cubic>]}
          BlueCurve {keys [<at 0 1 RCIM_Cubic>,<at 0.6468 1>,<at 1 1 RCIM_Cubic>]}
          GreenCurve {keys [<at 0 0.615 RCIM_Cubic>,<at 0.6468 0.3268>,<at 1 0.7525 RCIM_Cubic>]}
          RedCurve {keys [<at 0 0 RCIM_Cubic>,<at 0.6468 0.3054>,<at 1 0.441 RCIM_Cubic>]}
         }})
       LerpFactor=$User.LeftRightHand
      )
      Scale Factor=<1 1 1 1>
     )
     Float=&FloatFromCurve(
      FloatCurve=DataInterface() {propertyValues {Curve {keys [
          {arriveTangent=7.456, interpMode=RCIM_Cubic, leaveTangent=7.456, tangentMode=RCTM_User, time=0, value=0}
          {interpMode=RCIM_Cubic, tangentMode=RCTM_User, time=0.1, value=1}
          {arriveTangent=-3.2999, interpMode=RCIM_Cubic, leaveTangent=-3.2999, tangentMode=RCTM_User, time=1, value=0}
         ]}}}
      Scale Curve=0.4
     )
    )
    Create Color Attribute=true
    Create Scale Attribute=true
    Scale=&Multiply_VectorByFloat(
     Vector=&VectorFromCurve(
      VectorCurve=DataInterface() {propertyValues {
        XCurve {keys [
          {arriveTangent=1.3304, interpMode=RCIM_Cubic, leaveTangent=1.3304, tangentMode=RCTM_User, time=0, value=0}
          {interpMode=RCIM_Cubic, tangentMode=RCTM_User, time=1, value=1}
         ]}
        YCurve {keys [
          {arriveTangent=1.2179, interpMode=RCIM_Cubic, leaveTangent=1.2179, tangentMode=RCTM_User, time=0, value=0.2}
          {interpMode=RCIM_Cubic, tangentMode=RCTM_User, time=1, value=1}
         ]}
        ZCurve {keys [
          {arriveTangent=1.2179, interpMode=RCIM_Cubic, leaveTangent=1.2179, tangentMode=RCTM_User, time=0, value=0.2}
          {interpMode=RCIM_Cubic, tangentMode=RCTM_User, time=1, value=1}
         ]}
       }}
      Scale Curve=<0.75 0.5 0.5>
     )
     Float=&Multiply_Float(A=$System.ExplosionScale, B=1)
    )
    Write to Module Namespace=true
   ) {module=FN_Mesh_Attributes001}
   /SetVariables(Particles.ShockWaveDynmat:Vector4f=&MakeVector4(
     X=&Lerp_Float(
      A=1
      B=&FN_PickFloatDependOnPlatform(TargetPlatformActive=0.75, TargetPlatformNoActive=1)
      Alpha=&FloatFromCurve(FloatCurve=DataInterface() {propertyValues {Curve {keys [
           {interpMode=RCIM_Cubic, tangentMode=RCTM_User, time=0, value=1}
           {arriveTangent=-1.8092, interpMode=RCIM_Cubic, leaveTangent=-1.8092, tangentMode=RCTM_User, time=1, value=0}
          ]}}})
     )
     Y=0
     Z=1
     W=1
    )) {isSetParameters=true}
   &FN_Mesh_Attributes(
    Color=&MakeLinearColorFromVectorAndFloat(
     "Vector (RGB)"=<1 1 1>
     "Float (Alpha)"=&FloatFromCurve(
      FloatCurve=DataInterface() {propertyValues {Curve {keys [
          <at 0 0 RCIM_Cubic>
          {arriveTangent=0.1095, interpMode=RCIM_Cubic, leaveTangent=0.1095, tangentMode=RCTM_User, time=0.2, value=1}
          {arriveTangent=-1.5546, interpMode=RCIM_Cubic, leaveTangent=-1.5546, tangentMode=RCTM_User, time=1, value=0}
         ]}}}
      Scale Curve=0.5
     )
    )
    Create Color Attribute=true
    Create Orientation Attribute=true
    Create Scale Attribute=true
    Orientation:Quat4f=&OrientMeshToVector(Look At Direction=<0 0 1>, Reference Axis Vector=<1 0 0>)
    Scale=&Multiply_Vector(A=$Particles.Scale, B=<0.5 0.5 0.5>)
    Write to Module Namespace=true
   )
   /SetVariables(Particles.WindFlowDynmat:Vector4f=&MakeVector4(X=1, Y=1, Z=0, W=0)) {isSetParameters=true}
  ]
  rendererClasses [&NiagaraMeshRendererProperties]
  renderers [
   {rendererClass=&NiagaraMeshRendererProperties, rendererIndex=0}
   {rendererClass=&NiagaraMeshRendererProperties, rendererIndex=1}
   {rendererClass=&NiagaraMeshRendererProperties, rendererIndex=2}
  ]
 }
}
summary {
 systemName=NS_PC_PolarityWinn_Impact
 emitters [
  {emitterName=NE_Sprites, rendererClasses [&NiagaraSpriteRendererProperties]}
  {emitterName=NE_DarkFlare, rendererClasses [&NiagaraSpriteRendererProperties]}
  {emitterName=NE_SoundRings, rendererClasses [&NiagaraSpriteRendererProperties, &NiagaraLightRendererProperties]}
  {emitterName=NE_VolumeSquare, rendererClasses [&NiagaraMeshRendererProperties]}
  {
   emitterName=Light001
   rendererClasses [
    &NiagaraSpriteRendererProperties
    &NiagaraLightRendererProperties
    &NiagaraComponentRendererProperties
   ]
  }
 ]
 userVariables [
  User.ExplosionDelay {type=NiagaraFloat, defaultValue=0, description=""}
  User.LeftRightHand {type=NiagaraFloat, defaultValue=1, description=""}
  User.System_ExplosionScale {type=NiagaraFloat, defaultValue=100, description=""}
 ]
}
