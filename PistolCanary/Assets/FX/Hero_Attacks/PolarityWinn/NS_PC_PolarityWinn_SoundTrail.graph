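# Niagara trail effect for the PolarityWinn hero attack, leaving sound-wave rings and smoke along the moving system's path, with user bools for sliding and fade-out.
# NE_SoundWave spawns by distance, growing blue-to-purple sprites in local-space facing, with alpha scaled by System.FadeOut.
# NE_SmokePuff spawns by rate while IsSliding and by distance using SlidingAlpha as probability.
# Puffs shoot randomly left or right, slow under drag, rise, grow and fade through dynamic material parameters and curves.
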NiagaraSystem /PistolCanary/Assets/FX/Hero_Attacks/PolarityWinn/NS_PC_PolarityWinn_SoundTrail
dependencies {
 usedDataInterfaces [&NiagaraDataInterfaceCurve, &NiagaraDataInterfaceColorCurve]
 usedDynamicInputs [
  &Multiply_LinearColorByFloat
  &ClampFloat
  &Subtract_Float
 ]
 usedModules [
  &self:SystemSpawnScript.SystemScriptSource.SystemScriptGraph.NiagaraNodeAssignment_2.SetVariables_#1
  &SystemState
  &Timeline
  &self:SystemSpawnScript.SystemScriptSource.SystemScriptGraph.NiagaraNodeAssignment_1.SetVariables_#2
  &self:SystemSpawnScript.SystemScriptSource.SystemScriptGraph.NiagaraNodeAssignment_0.SetVariables_#3
 ]
}
emitters {
 NE_SmokePuff {
  emitterUpdateScript [
   &EmitterState()
   &SpawnRate(
    Spawn Probability=&MakeCustomFloatFromBool(Bool=$User.IsSliding, True Float=1, False Float=0)
    SpawnRate=4
    Use Spawn Probability=true
   )
   &SpawnPerUnit(
    Module.Use Max Movement Threshold=false
    Module.Use Movement Tolerance=false
    Spawn Probability:NiagaraFloat=$System.SlidingAlpha
    Spawn Spacing=50
    Use Spawn Probability=true
   )
  ]
  particleSpawnScript [
   /SetVariables(Particles.LeftRightFloat:NiagaraFloat=&UniformAOrBFloat(A=0, Evaluation Type=::NewEnumerator0, Recalculate Random Each Loop=true, Override Seed=false)) {isSetParameters=true}
   &InitializeParticle(
    Color=<rgba 0.6376 0.5841 0.5333 0.302>
    Lifetime Max=1
    Lifetime Min=0.7
    Lifetime Mode=::NewEnumerator1
    Mesh Scale Mode=::NewEnumerator3
    Mesh Uniform Scale=150
    Position Offset=&Lerp_Vector(A=<0 -75 0>, B=<0 75 0>, Alpha=$Particles.LeftRightFloat)
    Position Offset Coordinate Space=::Local
    Sprite Rotation Mode=::NewEnumerator2
    Sprite Size Mode=::NewEnumerator1
    Uniform Sprite Size Max=12
    Uniform Sprite Size Min=10
    UsePositionOffset=true
    Write Material Random=true
   ) {module=InitializeParticle002}
   &AddVelocity(
    Velocity=&Lerp_Vector(A=<0 -1 0>, B=<0 1 0>, Alpha=$Particles.LeftRightFloat)
    Velocity Speed Scale=&RandomRangeFloat(Minimum=950, Maximum=1150)
   )
  ]
  particleUpdateScript [
   &ParticleState()
   &GravityForce(Gravity=<0 0 200>)
   &Drag(Drag=6, Use Linear Drag=true, Use Rotational Drag=false)
   &SolveForcesAndVelocity(Clamp Velocity=false, Limit Acceleration=false)
   &ScaleColor(
    Scale Alpha=&FloatFromCurve(
     FloatCurve=DataInterface() {propertyValues {Curve {keys [
         {arriveTangent=7.3326, interpMode=RCIM_Cubic, leaveTangent=7.3326, tangentMode=RCTM_User, time=0, value=0}
         {arriveTangent=0.0071, interpMode=RCIM_Cubic, leaveTangent=0.0071, tangentMode=RCTM_User, time=0.4, value=1}
         <at 1 0 RCIM_Cubic>
        ]}}}
     Scale Curve=0.3
    )
    Scale RGB=&MakeVectorFromLinearColorRGB(Color=&ColorFromCurve(ColorCurve=DataInterface() {propertyValues {
        AlphaCurve {keys [<at 0 1 RCIM_Cubic>,<at 1 0 RCIM_Cubic>]}
        BlueCurve {keys [<at 0 1.2 RCIM_Cubic>,<at 0.4269 1 RCIM_Cubic>]}
        GreenCurve {keys [<at 0 1.2 RCIM_Cubic>,<at 0.4269 1 RCIM_Cubic>]}
        RedCurve {keys [<at 0 1.2 RCIM_Cubic>,<at 0.4269 1 RCIM_Cubic>]}
       }}))
    ScaleA=true
    ScaleRGB=true
   ) {module=ScaleColor001}
   &DynamicMaterialParameters(
    Index 0 Param 1=&Add_Float(
     A=&FloatFromCurve(FloatCurve=DataInterface() {propertyValues {Curve {keys [{arriveTangent=2.1549, time=0, value=0.8}, <at 1 1>]}}})
     B=&Multiply_Float(A=$Particles.Initial.MaterialRandom, B=0.5)
    )
    Index 1 Param 1=1
    Index 1 Param 2=1
    Index 1 Param 3=1
    Index 1 Param 4=1
    Index 2 Param 1=&FloatFromCurve(FloatCurve=DataInterface() {propertyValues {Curve {keys [
         {arriveTangent=1.8288, interpMode=RCIM_Cubic, leaveTangent=1.8288, tangentMode=RCTM_User, time=0, value=0.3}
         {arriveTangent=-0.0004, interpMode=RCIM_Cubic, leaveTangent=-0.0004, tangentMode=RCTM_User, time=1, value=1}
        ]}}})
    Index 2 Param 2=&RandomRangeFloat(Minimum=0.5, Maximum=1)
    Index 2 Param 3=4
    Index 2 Param 4=&RandomRangeFloat(Minimum=0.8, Maximum=1.2)
    Module.Param0WriteEnabled=true
    Module.Param0WriteEnabled1=true
    Module.Param0WriteEnabled2=true
    Module.Param0WriteEnabled3=true
    Module.Param1WriteEnabled=true
    Module.Param1WriteEnabled1=true
    Module.Param1WriteEnabled2=true
    Module.Param1WriteEnabled3=true
    Module.Param2WriteEnabled=true
    Module.Param2WriteEnabled1=true
    Module.Param2WriteEnabled2=true
    Module.Param2WriteEnabled3=true
    Module.Param3WriteEnabled=true
    Module.Param3WriteEnabled1=true
    Module.Param3WriteEnabled2=true
    Module.Param3WriteEnabled3=true
    Write Parameter Index 1=true
    Write Parameter Index 2=true
    Write Parameter Index 3=true
   ) {module=DynamicMaterialParameters002}
   &ScaleSpriteSize(Scale Factor=&Vector2DFromFloat(Value=&FloatFromCurve(FloatCurve=DataInterface() {propertyValues {Curve {keys [
          {arriveTangent=87.4549, interpMode=RCIM_Cubic, leaveTangent=87.4549, tangentMode=RCTM_User, time=0, value=1}
          {arriveTangent=4.8897, interpMode=RCIM_Cubic, leaveTangent=4.8897, tangentMode=RCTM_User, time=0.1, value=8}
          {interpMode=RCIM_Cubic, leaveTangent=-6.7553, time=1.0011, value=10}
         ]}}})))
   &CameraOffset(Camera Offset Amount=50)
   &SpriteRotationRate(Rotation Rate=&RandomRangeFloat(Minimum=-90, Maximum=90))
   &SpriteRotationRate(Rotation Rate=&Multiply_Float(A=90, B=&RandomRangeFloat(Minimum=-1, Maximum=1))) {module=SpriteRotationRate001}
  ]
  rendererClasses [&NiagaraSpriteRendererProperties]
  renderers [
   {rendererClass=&NiagaraSpriteRendererProperties}
   {rendererClass=&NiagaraSpriteRendererProperties, rendererIndex=1}
  ]
 }
 NE_SoundWave {
  emitterUpdateScript [
   &EmitterState()
   &SpawnPerUnit(Module.Use Max Movement Threshold=false, Module.Use Movement Tolerance=false, Spawn Spacing=500, Use Spawn Probability=false)
  ]
  particleSpawnScript [&InitializeParticle(
    Color Mode=::NewEnumerator0
    Mesh Scale Mode=::NewEnumerator3
    Sprite Size Mode=::NewEnumerator3
    Uniform Sprite Size=1000
    UsePositionOffset=false
    Write Material Random=true
   )]
  particleUpdateScript [
   &ParticleState()
   &ScaleSpriteSize(
    Uniform Curve Scale=2
    Uniform Curve Sprite Scale=DataInterface() {propertyValues {Curve {keys [
        {arriveTangent=0.0194, interpMode=RCIM_Cubic, leaveTangent=0.0194, tangentMode=RCTM_User, time=0, value=0.5}
        {interpMode=RCIM_Cubic, tangentMode=RCTM_User, time=1, value=1.475}
       ]}}}
   )
   &DynamicMaterialParameters(
    Index 0 Param 1=0.2
    Index 0 Param 2=1
    Module.Index0Float0Write=true
    Module.Index0Float1Write=true
    Module.Index0Float2Write=true
    Module.Index0Float3Write=true
   )
   &ScaleColor(
    Scale Alpha=&FloatFromCurve(
     FloatCurve=DataInterface() {propertyValues {Curve {keys [
         {arriveTangent=4.2862, interpMode=RCIM_Cubic, leaveTangent=4.2862, tangentMode=RCTM_User, time=0, value=0}
         {arriveTangent=-1.1812, interpMode=RCIM_Cubic, leaveTangent=-1.1812, tangentMode=RCTM_User, time=0.1, value=1}
         {arriveTangent=-0.0521, interpMode=RCIM_Cubic, leaveTangent=-0.0521, tangentMode=RCTM_User, time=1, value=0}
        ]}}}
     Scale Curve=&Multiply_Float(A=0.5, B=$System.FadeOut)
    )
    Scale RGB=&MakeVectorFromLinearColorRGB(Color=&ColorFromCurve(ColorCurve=DataInterface() {propertyValues {
        AlphaCurve {keys [<at 0 1 RCIM_Cubic>]}
        BlueCurve {keys [<at 0.0375 0.8 RCIM_Cubic>,<at 0.2712 1>,<at 0.5921 1 RCIM_Cubic>]}
        GreenCurve {keys [<at 0.0375 0.4306 RCIM_Cubic>,<at 0.2712 0>,<at 0.5921 0.2554 RCIM_Cubic>]}
        RedCurve {keys [<at 0.0375 0.08 RCIM_Cubic>,<at 0.2712 0.2184>,<at 0.5921 0.4442 RCIM_Cubic>]}
       }}))
    ScaleA=true
    ScaleRGB=true
   )
   &SpriteFacingAndAlignment(Facing Coordinate Space=::Local, WriteAlignment=false, WriteFacing=true)
   &FN_Mesh_Attributes(
    Color=&MultiplyLinearColor(
     Linear Color=&ColorFromCurve(ColorCurve=DataInterface() {propertyValues {
        AlphaCurve {keys [<at 0 1 RCIM_Cubic>]}
        BlueCurve {keys [<at 0 1 RCIM_Cubic>,<at 0.9865 1>]}
        GreenCurve {keys [<at 0 0.6758 RCIM_Cubic>,<at 0.9865 0.4228>]}
        RedCurve {keys [<at 0 0 RCIM_Cubic>,<at 0.9865 0.8464>]}
       }})
     Scale Factor=&MakeVector4FromVectorAndFloat(
      "Vector (RGB)"=&VectorFromFloat(Value=1)
      "Float (Alpha)"=&FloatFromCurve(
       FloatCurve=DataInterface() {propertyValues {Curve {keys [
           {interpMode=RCIM_Cubic, tangentMode=RCTM_User, time=0, value=1}
           {arriveTangent=-1.6975, interpMode=RCIM_Cubic, leaveTangent=-1.6975, tangentMode=RCTM_User, time=1, value=0}
          ]}}}
       Scale Curve=5
      )
     )
    )
    Create Color Attribute=true
    Create Scale Attribute=true
    Scale=&VectorFromFloat(Value=&Multiply_Float(A=/Niagara/DynamicInputs/TypeConversions/Vector2DLength(Vector:Vector2f=$Particles.SpriteSize), B=0.004))
    Write to Module Namespace=true
   ) {module=FN_Mesh_Attributes001}
   /SetVariables(Particles.ShockWaveDynmat:Vector4f=&MakeVector4(
     X=&Lerp_Float(
      A=1
      B=0.5
      Alpha=&FloatFromCurve(FloatCurve=DataInterface() {propertyValues {Curve {keys [
           {interpMode=RCIM_Cubic, tangentMode=RCTM_User, time=0, value=1}
           {arriveTangent=-1.8092, interpMode=RCIM_Cubic, leaveTangent=-1.8092, tangentMode=RCTM_User, time=1, value=0}
          ]}}})
     )
     Y=0
     Z=0
     W=1
    )) {isSetParameters=true}
   &FN_Sprite_Attributes(
    Create SpriteSize Attribute=true
    SpriteSize=&Multiply_Vector2D_ByFloat(
     Float=&FloatFromCurve(FloatCurve=DataInterface() {propertyValues {Curve {keys [<at 0 1>,<at 0.5 0>]}}}, Scale Curve=0.5)
     Vector2D=$Particles.SpriteSize
    )
    Write to Module Namespace=true
   )
  ]
  rendererClasses [&NiagaraSpriteRendererProperties]
  renderers [
   {rendererClass=&NiagaraSpriteRendererProperties}
   {rendererClass=&NiagaraSpriteRendererProperties, rendererIndex=1}
  ]
 }
}
summary {
 systemName=NS_PC_PolarityWinn_SoundTrail
 emitters [
  {emitterName=NE_SoundWave, rendererClasses [&NiagaraSpriteRendererProperties]}
  {emitterName=NE_SmokePuff, rendererClasses [&NiagaraSpriteRendererProperties]}
 ]
 userVariables [
  User.FadeOut {type=NiagaraBool}
  User.IsSliding {type=NiagaraBool}
 ]
}
